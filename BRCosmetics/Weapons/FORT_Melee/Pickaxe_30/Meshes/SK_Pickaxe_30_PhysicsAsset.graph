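# Physics asset for the SK_Pickaxe_30 skeletal mesh, a cosmetic two-handed melee pickaxe.
# A single box body on the Melee_TwoHanded_Handle bone, offset and rotated from the bone origin, provides its collision shape.
# It has no constraints, so the asset defines no jointed or articulated parts.
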PhysicsAsset /BRCosmetics/Weapons/FORT_Melee/Pickaxe_30/Meshes/SK_Pickaxe_30_PhysicsAsset
bodies {Melee_TwoHanded_Handle [{shapeType=Box, extentX=109.3008, extentY=23.0015, extentZ=142.5272, center=<0.1335 0.218 10.8153>, rotation=<rot -1.0945 -90 -0.1762>}]}
constraints []
